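# Physics asset for the Bandolier pickaxe, a two-handed melee cosmetic, defining collision bodies and joints for its dangling cloth pieces.
# Capsules cover the handle, knife and small cloth bones, while boxes shape the larger cloth pieces.
# Two cloth branches hang from the handle and two from the knife, each branch being a one- or two-bone chain with limited swing and twist.
# No constraint links the knife to the handle.
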PhysicsAsset /BRCosmetics/Weapons/FORT_Melee/Pickaxe_Bandolier/Meshes/Pickaxe_Bandolier_PhysicsAsset
bodies {
 Melee_TwoHanded_Handle [{shapeType=Capsule, length=101.2087, radius=4.2422, center=<0.5732 -0.0358 10.4026>, rotation=<rot 0.0699 -90.0013 0.142>}]
 dyn_Clonth_1_01 [{shapeType=Capsule, length=1.5, radius=1.75, center=<-0.2153 0 0.0482>, rotation=<rot -81.5263 -179.9999 -179.9998>}]
 dyn_Clonth_1_02 [
  {
   shapeType=Box
   extentX=1.8717
   extentY=2.6422
   extentZ=8.5624
   shapeName=dyn_Clonth_1_02_box
   center=<4.0979 -1.8905 7.2467>
   rotation=<rot -23.1647 -91.414 11.1677>
  }
  {
   shapeType=Box
   extentX=7.5
   extentY=2
   extentZ=4.5
   shapeName=dyn_Clonth_1_02_box
   center=<-0.1469 -0.0327 0.101>
   rotation=<rot 11.314 7.4686 -0.9537>
  }
 ]
 dyn_Clonth_2 [{
   shapeType=Box
   extentX=1.5217
   extentY=4.7873
   extentZ=10.762
   shapeName=dyn_Clonth_2_box
   center=<5.504 -0.3752 0.8345>
   rotation=<rot -9.3142 -89.9998 43.7591>
  }]
 dyn_Knife [{shapeType=Capsule, length=49.0954, radius=5.1436, center=<0.4241 -14.9099 -1.0524>, rotation=<rot 0 172.448 -83.5908>}]
 dyn_KnifeClonth_1_01 [{shapeType=Capsule, length=3.25, radius=0.75, center=<0.8977 0 -0.1341>, rotation=<rot -86.2642 0.0001 0>}]
 dyn_KnifeClonth_1_02 [{shapeType=Capsule, length=5.25, radius=0.75, center=<0.964 0 -0.3774>, rotation=<rot -88.3021 -0.0034 0.0033>}]
 dyn_KnifeClonth_2 [{shapeType=Capsule, length=3, radius=0.75, center=<1.2136 0 -0.1821>, rotation=<rot -81.4658 179.9994 -179.9994>}]
}
constraints [
 {bone1Name=dyn_KnifeClonth_1_01, bone2Name=dyn_Knife, swing1LimitDegrees=40, swing2LimitDegrees=40, twistLimitDegrees=20}
 {bone1Name=dyn_KnifeClonth_1_02, bone2Name=dyn_KnifeClonth_1_01, swing1LimitDegrees=40, swing2LimitDegrees=40, twistLimitDegrees=20}
 {bone1Name=dyn_KnifeClonth_2, bone2Name=dyn_Knife, swing1LimitDegrees=80, swing2LimitDegrees=80, twistLimitDegrees=20}
 {bone1Name=dyn_Clonth_1_01, bone2Name=Melee_TwoHanded_Handle, swing1LimitDegrees=23, swing2LimitDegrees=23, twistLimitDegrees=20}
 {bone1Name=dyn_Clonth_1_02, bone2Name=dyn_Clonth_1_01, swing1LimitDegrees=15, swing2LimitDegrees=15, twistLimitDegrees=20}
 {bone1Name=dyn_Clonth_2, bone2Name=Melee_TwoHanded_Handle, swing1LimitDegrees=24, swing2LimitDegrees=24, twistLimitDegrees=20}
]
